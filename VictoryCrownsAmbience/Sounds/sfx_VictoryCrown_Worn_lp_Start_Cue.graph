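# Sound cue for the start of the Victory Crown worn-loop ambience, assigned to the Fort_UI sound class.
# A distance crossfade hands off from a first-person wave up close to a third-person wave at mid range, which fades out farther away.
# A modulator at the root adds very slight pitch variation.
# Overridden attenuation shapes the falloff, and overridden concurrency allows one instance, blocking new plays.
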SoundCue /VictoryCrownsAmbience/Sounds/sfx_VictoryCrown_Worn_lp_Start_Cue
root=Modulator_1
nodes [
 {
  id=DistanceCrossFade_0
  children [WavePlayer_6,WavePlayer_7]
  properties {crossFadeInput [
    {fadeInDistanceEnd=0, fadeInDistanceStart=0, fadeOutDistanceEnd=2000, fadeOutDistanceStart=1000, volume=1}
    {fadeInDistanceEnd=1500, fadeInDistanceStart=1350, fadeOutDistanceEnd=6500, fadeOutDistanceStart=5800, volume=1}
   ]}
 }
 {id=Modulator_1, children [DistanceCrossFade_0], properties {pitchMax=1.005, pitchMin=1}}
 {id=WavePlayer_6, properties {soundWaveAssetPtr=/VictoryCrownsAmbience/Sounds/wav/sfx_VictoryCrown_Worn_1P_lp_Start}}
 {id=WavePlayer_7, properties {soundWaveAssetPtr=/VictoryCrownsAmbience/Sounds/wav/sfx_VictoryCrown_Worn_3P_lp_Start}}
]
properties {
 bOverrideAttenuation=true
 bOverrideConcurrency=true
 maxDistance=7500
 soundClassObject=&Fort_UI
 volumeMultiplier=0.9
 attenuationOverrides {falloffDistance=7000, attenuationShapeExtents {x=500}}
 concurrencyOverrides {maxCount=1, resolutionRule=PreventNew}
}
